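# Niagara skydiving contrail cosmetic that adds banana meshes and ribbon trails to the character's skeletal mesh.
# BananaBunch pins two persistent meshes to bones, while UpwardBananaBurst continuously launches spinning bananas upward with gravity, drag and a wobbly scale pop.
# Local_Space_Skydive draws two or four ribbons by quality level, with bone-locked lead particles and partly inherited source movement.
# The dive amount shapes launch direction, ribbon spread, movement inheritance and opacity scale.
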NiagaraSystem /BRCosmetics/Athena/Effects/Particles/Skydiving/NS_Contrail_Bananas/NS_Contrails_Banana_Converted
dependencies {
 usedDataInterfaces [&NiagaraDataInterfaceMeshRendererInfo, &NiagaraDataInterfaceCurve]
 usedDynamicInputs [
  &SetBoolByFloatComparison
  &self:Abs_Z_Axis_01
  &TransformVector
  &FN_FrontEndFloatModifier
  &NormalizeVector
  &Multiply_Vector
 ]
 usedModules [
  &SystemState
  &self:SystemSpawnScript.SystemScriptSource.SystemScriptGraph.NiagaraNodeAssignment_0.SetVariables_#1
  &Timeline
  &self:SystemSpawnScript.SystemScriptSource.SystemScriptGraph.NiagaraNodeAssignment_3.SetVariables_#2
  &self:SystemSpawnScript.SystemScriptSource.SystemScriptGraph.NiagaraNodeAssignment_4.SetVariables_#3
  /Game/Effects/Niagara/Modules/Skydiving/FN_DivingAttributes/v5/FN_DivingAttributes
  &self:SystemSpawnScript.SystemScriptSource.SystemScriptGraph.NiagaraNodeAssignment_5.SetVariables_#4
  &FN_Integrator
 ]
}
emitters {
 BananaBunch {
  emitterUpdateScript [
   &EmitterState(Life Cycle Mode=::NewEnumerator1, Loop Behavior=::NewEnumerator1, Loop Duration Mode=::NewEnumerator1, UseLoopDelay=false)
   &SpawnBurst_Instantaneous(Spawn Count=2, Spawn Group=1, Use Spawn Probability=false, UseLoopCountLimit=false)
  ]
  particleSpawnScript [
   &PartitionParticles()
   &InitializeParticle(Mesh Scale Mode=::NewEnumerator3, UsePositionOffset=false, Write Material Random=true) {module=InitializeParticle001}
   &InitialMeshOrientation(Mesh Axis To Orient=<0 0 0>, Mesh Orientation Mode=::NewEnumerator2, Orientation Vector=<0 0 0>, Use Rotation Vector=false)
  ]
  particleUpdateScript [
   &ParticleState(
    Kill Particles When Lifetime Has Elapsed=false
    Let Infinitely Lived Particles Die When Emitter Deactivates=true
    Loop Particles Lifetime=true
   )
   &SkeletalMeshLocation(
    Bone / Socket Index=$Particles.PartitionParticles.Partition
    Bone Sampling Mode=::NewEnumerator4
    Inherit Velocity=false
    MaskSpawnGroup=false
    Sampled Position Offset=<0 0 2>
    Sampled Position Offset Coordinate Space:ENiagaraMeshTransforms=::NewEnumerator1
    Skeletal Mesh=$User.SkeletalMesh
   )
   &SolveForcesAndVelocity(Clamp Velocity=false, Limit Acceleration=false)
  ]
  rendererClasses [&NiagaraMeshRendererProperties]
  renderers [{rendererClass=&NiagaraMeshRendererProperties, rendererIndex=0}]
 }
 Local_Space_Skydive {
  emitterSpawnScript [/SetVariables(Emitter.Number of Trails=HlslExpression() {hlslExpression="Engine.QualityLevel > 2 ? 4 : 2"}) {isSetParameters=true}]
  emitterUpdateScript [
   &EmitterState(Life Cycle Mode=::NewEnumerator1, Loop Duration=1, UseLoopDelay=false)
   &SpawnBurst_Instantaneous(
    Spawn Count=$Emitter.Initial.Number of Trails
    Spawn Group=-1
    Spawn Probability=HlslExpression() {hlslExpression="Emitter.LoopCount == 0"}
    Use Spawn Probability=true
   )
   &SpawnRate(SpawnRate=60, Use Spawn Probability=false)
  ]
  particleSpawnScript [
   &FN_SetSpawnGroup(Number of Spawn Groups:NiagaraInt32=$Emitter.Initial.Number of Trails)
   &self:LockParticlesToBones_01() {module=LockParticlesToBones_01}
   &InitializeRibbon(
    Lifetime=&MakeCustomFloatFromBool(Bool=$Particles.Locked, True Float=100, False Float=/Niagara/DynamicInputs/Math/Min_Float(A:NiagaraFloat=$Emitter.Age, B:NiagaraFloat=0.3))
    Module.Write Color=true
    Module.Write Lifetime=true
    Module.Write Mass=true
    Module.Write Material Random=true
    Module.Write Position=false
    Module.Write RibbonTwist=false
    Module.Write RibbonWidth=true
    Ribbon Width=20
   )
   &FN_FlipRibbonV() {enabled=false}
   &SkeletalMeshLocation(
    Bone / Socket Index=$Particles.SpawnGroup
    Bone Sampling Mode=::NewEnumerator4
    Inherit Velocity=false
    MaskSpawnGroup=false
    Skeletal Mesh=$User.SkeletalMesh
   )
   /Game/Effects/Niagara/Modules/Skydiving/FN_Skydive_Velocity_v4/FN_Skydive_Velocity(
    Bone Orientation:Quat4f=$Output.SkeletalMeshLocation.SampledRotation
    Bone Position:Vector3f=$Output.SkeletalMeshLocation.SampledPosition
    FE Speed=1103.7715
    FE Spread:NiagaraFloat=/Game/Effects/Niagara/DynamicInputs/Skydiving/SkydiveHandsFeetFloatModifier(Hands:NiagaraFloat=&Lerp_Float(A=0, B=200, Alpha=$System.Dive), Feet:NiagaraFloat=0)
    SpawnGroup:NiagaraInt32=$Particles.SpawnGroup
    Speed=221.3639
    Spread=148.0048
   ) {module=FN_Skydive_Velocity001}
   /Game/Effects/Niagara/Modules/Skydiving/FN_SkydivePositionOffset_v1/FN_SkydivePositionOffset(
    Bone Orientation:Quat4f=$Output.SkeletalMeshLocation.SampledRotation
    Bone Position:Vector3f=$Output.SkeletalMeshLocation.SampledPosition
    Offset:NiagaraFloat=$Particles.OffsetAmount
   )
  ]
  particleUpdateScript [
   &ParticleState()
   /SetVariables(Particles.Age:NiagaraFloat=&MakeCustomFloatFromBool(Bool=$Particles.Locked, True Float=0, False Float=$Particles.Age)) {isSetParameters=true}
   /SetVariables(Particles.NormalizedAge:NiagaraFloat=&MakeCustomFloatFromBool(Bool=$Particles.Locked, True Float=0, False Float=$Particles.NormalizedAge)) {isSetParameters=true}
   &SkeletalMeshLocation(
    Apply To Particle Position=false
    Bone / Socket Index=$Particles.SpawnGroup
    Bone Sampling Mode=::NewEnumerator4
    Inherit Velocity=false
    Interpolate Mesh Sampling=false
    MaskSpawnGroup=false
    Skeletal Mesh=$User.SkeletalMesh
   ) {module=SkeletalMeshLocation001}
   &FN_SkydivePositionOffset(
    Bone Orientation:Quat4f=$Output.SkeletalMeshLocation001.SampledRotation
    Bone Position:Vector3f=$Output.SkeletalMeshLocation001.SampledPosition
    Offset:NiagaraFloat=$Particles.OffsetAmount
    Write to Intrinsic Properties=false
   ) {module=FN_SkydivePositionOffset001}
   &InheritSourceMovement(
    Applied Position Scale=&VectorFromFloat(Value=&Lerp_Float(A=0.6, B=0.9, Alpha=$System.Dive))
    ApplyForce=false
    ApplyPosition=true
    ApplyPositionFalloff=false
   )
   &SolveForcesAndVelocity(Clamp Velocity=false, Limit Acceleration=false)
   /SetVariables(Particles.RenderPosition:Vector3f=&MakeCustomVectorFromBool(
     Bool=$Particles.Locked
     True Vector=&Add_Vector(A=$Output.FN_SkydivePositionOffset001.OffsetDirection, B=$Output.SkeletalMeshLocation001.SampledPosition)
     False Vector=Linked() {linkedVariable {name=Particles.Position, type=NiagaraPosition}}
    )) {isSetParameters=true}
   &FN_RibbonMaterial(
    Opacity Pan U Offset:NiagaraFloat=/Game/Effects/Niagara/DynamicInputs/Random/FN_SeededRandomRangeFloat(Minimum:NiagaraFloat=0, Maximum:NiagaraFloat=1, Random Seed:NiagaraInt32=$Particles.SpawnGroup)
    Opacity Pan U Speed=1.6096
    Opacity Power=0.35
    Opacity Scale U:NiagaraFloat=&Multiply_Float(
     A=&FN_SeededRandomRangeFloat(Minimum:NiagaraFloat=0.7, Maximum:NiagaraFloat=1.01, Random Seed:NiagaraInt32=$Particles.SpawnGroup)
     B=&Lerp_Float(A=1, B=0.2, Alpha=$System.Dive)
    )
    Scale Alpha:NiagaraFloat=&FloatFromCurve(
     FloatCurve=DataInterface() {propertyValues {Curve {keys [{tangentMode=RCTM_User, time=0, value=1}, {tangentMode=RCTM_User, time=1, value=0}]}}}
     Scale Curve=/Game/Effects/Niagara/DynamicInputs/Math/FN_Fit(
      Clamp:NiagaraBool=true
      Value:NiagaraFloat=$Engine.Owner.LODDistance
      SourceMin:NiagaraFloat=400
      SourceMax:NiagaraFloat=2000
      DestMin:NiagaraFloat=0.03
      DestMax:NiagaraFloat=0.2
     )
    )
    TIme:NiagaraFloat=$System.FN_Integrator.Value
    UV0 Scale=0.5
   )
  ]
  rendererClasses [&NiagaraRibbonRendererProperties]
  renderers [{rendererClass=&NiagaraRibbonRendererProperties}]
 }
 UpwardBananaBurst {
  emitterUpdateScript [
   &EmitterState()
   &SpawnRate(SpawnRate=4.5, Use Spawn Probability=false) {module=SpawnRate001}
  ]
  particleSpawnScript [
   &PartitionParticles()
   &InitializeParticle(
    Lifetime Max=2.5
    Lifetime Min=1.5
    Lifetime Mode=::NewEnumerator1
    Mesh Scale Mode=::NewEnumerator1
    Mesh Uniform Scale Max=0.4
    Mesh Uniform Scale Min=0.3
    UsePositionOffset=false
    Write Material Random=true
   )
   &SkeletalMeshLocation(
    Bone / Socket Index=$Particles.PartitionParticles.Partition
    Bone Sampling Mode=::NewEnumerator4
    MaskSpawnGroup=false
    Skeletal Mesh=$User.SkeletalMesh
    Spawn Group Mask=1
   )
   &InitialMeshOrientation(
    Mesh Orientation Mode=::NewEnumerator2
    Orientation Coordinate Space=::Local
    Orientation Vector=<0 1 0>
    Rotation Coordinate Space=::NewEnumerator1
    Use Rotation Vector=false
   ) {module=InitialMeshOrientation001}
   &AddVelocity(
    Cone Angle=30
    Cone Axis=&FN_FrontEndVector3Modifier(
     Gameplay=&Lerp_Vector(A=<0 0 1>, B=<0 -0.1 1>, Alpha=$System.Dive)
     Front End=&Lerp_Vector(A=<0 0 1>, B=<0 -0.25 1>, Alpha=$System.Dive)
    )
    Module.UseConeFalloff=true
    Module.UseVelDistribution=true
    Override Seed=false
    Speed Falloff From Cone Axis=0.875
    Velocity Mode=::NewEnumerator2
    Velocity Speed=&RandomRangeFloat(Minimum=300, Maximum=450)
   )
  ]
  particleUpdateScript [
   &ParticleState()
   &UpdateMeshOrientation(Rotation Rate=&FloatFromCurve(
     FloatCurve=DataInterface() {propertyValues {Curve {keys [{interpMode=RCIM_Cubic, tangentMode=RCTM_User, time=0, value=0}, {interpMode=RCIM_Cubic, tangentMode=RCTM_User, time=1, value=1}]}}}
     Scale Curve=2
    ))
   &ScaleMeshSize(Scale Factor=&VectorFromFloat(Value=&FloatFromCurve(FloatCurve=DataInterface() {propertyValues {Curve {keys [
          {arriveTangent=19.0214, interpMode=RCIM_Cubic, time=0, value=0}
          {arriveTangent=6.5269, interpMode=RCIM_Cubic, leaveTangent=6.5269, time=0.0492, value=1.7}
          {arriveTangent=-1.7532, interpMode=RCIM_Cubic, leaveTangent=-1.7532, time=0.0996, value=0.65}
          {arriveTangent=0.7945, interpMode=RCIM_Cubic, leaveTangent=0.7945, time=0.1633, value=1.5}
          {arriveTangent=-1.8445, interpMode=RCIM_Cubic, leaveTangent=-1.8445, time=0.2254, value=0.75}
          {arriveTangent=0.8561, interpMode=RCIM_Cubic, leaveTangent=0.8561, time=0.2988, value=1.25}
          {arriveTangent=-0.7148, interpMode=RCIM_Cubic, leaveTangent=-0.7148, time=0.3715, value=0.875}
          {arriveTangent=0.3232, interpMode=RCIM_Cubic, leaveTangent=0.3232, time=0.4737, value=1.125}
          {arriveTangent=-0.6244, interpMode=RCIM_Cubic, leaveTangent=-0.6244, time=0.5726, value=0.94}
          {arriveTangent=0.1404, interpMode=RCIM_Cubic, leaveTangent=0.1404, time=0.6739, value=1}
          <at 1 1>
         ]}}})))
   &GravityForce(Gravity=<0 0 -380>)
   &Drag(Use Linear Drag=true, Use Rotational Drag=true)
   &ScaleColor(
    Scale Alpha=&FloatFromCurve(FloatCurve=DataInterface() {propertyValues {Curve {keys [<at 0.5 1.0113 RCIM_Cubic>,<at 1.0011 0.002 RCIM_Cubic>]}}})
    ScaleA=true
    ScaleRGB=false
   ) {enabled=false}
   &Color(Color=&MakeLinearColorFromVectorAndFloat("Vector (RGB)"=<0.1 0.1 0.1>, "Float (Alpha)"=&ClampFloat(Float=1)))
   &SolveForcesAndVelocity(Clamp Velocity=false, Limit Acceleration=false)
   &DynamicMaterialParameters(
    Index 0 Param 1=&FloatFromCurve(FloatCurve=DataInterface() {propertyValues {Curve {keys [<at 0.2486 0 RCIM_Cubic>,<at 0.4982 1 RCIM_Cubic>]}}})
    Index 0 Param 2=&FloatFromCurve(FloatCurve=DataInterface() {propertyValues {Curve {keys [<at 0.6263 0 RCIM_Cubic>,<at 0.8496 0 RCIM_Cubic>]}}})
    Index 0 Param 3=&FloatFromCurve(FloatCurve=DataInterface() {propertyValues {Curve {keys [
         {arriveTangent=0.0051, interpMode=RCIM_Cubic, time=0, value=1}
         {arriveTangent=-1.3742, interpMode=RCIM_Cubic, leaveTangent=-1.3742, time=0.3428, value=0.9998}
         {arriveTangent=-1.5213, interpMode=RCIM_Cubic, leaveTangent=-1.5213, time=0.5956, value=0.1815}
         <at 1 0 RCIM_Cubic>
        ]}}})
    Index 0 Param 4=&FloatFromCurve(FloatCurve=DataInterface() {propertyValues {Curve {keys [<at 0 0>, {arriveTangent=1, interpMode=RCIM_Cubic, leaveTangent=1, time=0.75, value=0}, <at 1 1>]}}})
    Module.Param0WriteEnabled=true
    Module.Param1WriteEnabled=true
    Module.Param2WriteEnabled=true
    Module.Param3WriteEnabled=true
   )
  ]
  rendererClasses [&NiagaraMeshRendererProperties]
  renderers [{rendererClass=&NiagaraMeshRendererProperties, rendererIndex=0}]
 }
}
summary {
 systemName=NS_Contrails_Banana_Converted
 emitters [
  {emitterName=BananaBunch, rendererClasses [&NiagaraMeshRendererProperties]}
  {emitterName=UpwardBananaBurst, rendererClasses [&NiagaraMeshRendererProperties]}
  {emitterName=Local_Space_Skydive, rendererClasses [&NiagaraRibbonRendererProperties]}
 ]
 userVariables [
  User.bIsFrontEnd {type=NiagaraBool}
  User.bIsFrontEndPreview {type=NiagaraBool}
  User.blsFrontEndPreview {type=NiagaraBool, defaultValue=true}
  User.ParaGlideLeanAlpha {type=NiagaraFloat, defaultValue=0, description=""}
  User.SkeletalMesh {
   type=NiagaraDataInterfaceSkeletalMesh
   defaultValue {
    struct=NiagaraExt_VariableValue_DataInterface
    value {dataInterface=&self:NiagaraDataInterfaceSkeletalMesh_0}
   }
  }
 ]
}
